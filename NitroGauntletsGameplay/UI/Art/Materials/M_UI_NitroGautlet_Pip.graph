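# Translucent UI material that draws one Nitro Gauntlets pip with an adjustable fill progress.
# A four-vertex polygon SDF, with corners set by vector parameters, is thresholded into the pip shape and smoothstepped into a soft glow.
# A rotated linear gradient offset by Progress becomes a sharp fill mask.
# Named reroutes share these masks to blend empty color, pip color and glow, with separate opacities for empty and filled parts.
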Material /NitroGauntletsGameplay/UI/Art/Materials/M_UI_NitroGautlet_Pip
graph {
 properties {blendMode=BLEND_Translucent, materialDomain=MD_UI, pixelDepthOffsetMode=PDOM_Legacy, usedShadingModels=Default Lit}
 nodes [
  Add(
   A=Multiply(A=NamedRerouteUsage(), B=ScalarParameter() {defaultValue=0.2, parameterName=EmptyPartOpacity}) {constB=0.5}
   B=Multiply(A=NamedRerouteUsage(), B=ScalarParameter() {defaultValue=1, parameterName=FillOpacity})
  )
  n1 = MaterialFunctionCall!(
   UVs
   vertex_0=VectorParameter() {parameterName=A, defaultValue=<rgba -0.5 -0.6 0 1>}.RGB
   vertex_1=VectorParameter() {parameterName=B, defaultValue=<rgba 0.5 -0.8 0 1>}.RGB
   vertex_2=VectorParameter() {parameterName=C, defaultValue=<rgba 0.5 0.8 0 1>}.RGB
   vertex_3=VectorParameter() {parameterName=D, defaultValue=<rgba -0.5 0.6 0 1>}.RGB
   Glow Min
   Glow Max
   Outline Thickness=Constant()
  ) {materialFunction=&/MF_4VertexPolygon}
  n2 = Constant() {r=0.1}
  n3 = OneMinus(None=SmoothStep(Value=Subtract(A=@n1.SDF, B=@n2) {constB=0}) {constMax=0.02})
  n4 = VectorParameter() {parameterName=EmtyColor, defaultValue=<rgba 0 0 0 1>}
  n5 = Multiply(A=Multiply(A=@n3, B=VectorParameter() {parameterName=PipColor, defaultValue=<rgba 0.9647 0.2747 0 1>}.RGB), B=NamedRerouteUsage())
  n6 = NamedRerouteUsage()
  LinearInterpolate(
   A=LinearInterpolate(A=Multiply(A=@n3, B=@n4.RGB), B=@n5, Alpha=@n5)
   B=Multiply(A=@n6, B=VectorParameter() {parameterName=PipColor, defaultValue=<rgba 0.9647 0.2747 0 1>}.RGB)
   Alpha=@n6
  )
  NamedRerouteDeclaration(None=@n3) {name=Opacity, nodeColor=<rgba 0 0 0 1>}
  n7 = SmoothStep(Max=ScalarParameter() {defaultValue=0.3, parameterName=ProgressBlur}, Value=@n1.SDF) {constMax=0.3}
  NamedRerouteDeclaration(None=Multiply(A=OneMinus(None=@n7), B=NamedRerouteUsage())) {name=Glow, nodeColor=<rgba 0 0 0 1>}
  n8 = MaterialFunctionCall(RotationAngle=Constant() {r=0.25}, GradientOffset=ScalarParameter() {parameterName=Progress}) {materialFunction=&MF_UI_LinearGradient}
  NamedRerouteDeclaration(None=OneMinus(None=SmoothStep(Value=@n8.Result) {constMax=0.01})) {name=Progress, nodeColor=<rgba 0 0 0 1>}
 ]
}
